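# Looping outdoor ambience cue for a morning arid desert, in the Figment POI sounds, using Athena ambient attenuation and sound class.
# A quality-level root plays the desert morning loop on the first two levels and a mild daytime wind loop on the last two.
# Only the wind branch runs through a modulator, which boosts its volume and varies pitch slightly upward, so the ambience differs by quality level.
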SoundCue /FigmentCore/Sounds/POIs/SC_Arid_Desert_Morning_Figment_Cue
root=QualityLevel_0
nodes [
 {id=Modulator_0, children [WavePlayer_1], properties {pitchMax=1.1, pitchMin=1, volumeMax=2, volumeMin=2}}
 {id=QualityLevel_0, children [WavePlayer_0,WavePlayer_0,Modulator_0,Modulator_0]}
 {id=WavePlayer_0, properties {bLooping=true, soundWaveAssetPtr=/FigmentCore/Sounds/POIs/Wavs/Desert_Ambient_Morning_01}}
 {id=WavePlayer_1, properties {bLooping=true, soundWaveAssetPtr=/FigmentCore/Sounds/POIs/Wavs/mild_daytime_wind_01_opt}}
]
properties {
 attenuationSettings=&Fort_Ambient_Athena_Att
 maxDistance=2097152
 soundClassObject=&Fort_Ambient_Outdoors_Athena
 volumeMultiplier=0.5
}
