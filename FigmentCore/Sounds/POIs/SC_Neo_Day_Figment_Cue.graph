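# Sound cue for the Neo Tilted daytime ambience in FigmentCore's POI sounds, named as a loop.
# A quality level node at the root points all four of its inputs at one shared wave player, so each quality level plays the same wave.
# The cue uses the Athena ambient attenuation and outdoor ambient sound class, with its volume reduced.
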SoundCue /FigmentCore/Sounds/POIs/SC_Neo_Day_Figment_Cue
root=QualityLevel_0
nodes [
 {id=QualityLevel_0, children [WavePlayer_2,WavePlayer_2,WavePlayer_2,WavePlayer_2]}
 {id=WavePlayer_2, properties {soundWaveAssetPtr=/FigmentCore/Sounds/POIs/Wavs/NeoTilted_Ambient_Day_01_LOOP}}
]
properties {
 attenuationSettings=&Fort_Ambient_Athena_Att
 maxDistance=2097152
 soundClassObject=&Fort_Ambient_Outdoors_Athena
 volumeMultiplier=0.5
}
