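# Looping outdoor ambience cue for a temperate winter evening in Figment POIs, built from daytime winter and desert recordings.
# A quality-level node sends its first two slots to a mixer that layers a winter day loop with a quieter desert day loop.
# Its last two slots play a single winter loop variant suffixed "opt".
# The cue uses the Athena outdoor ambient sound class and Athena ambient attenuation.
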SoundCue /FigmentCore/Sounds/POIs/SC_Temperate_Winter_Evening_Figment_Cue
root=QualityLevel_0
nodes [
 {id=Mixer_0, children [WavePlayer_2,WavePlayer_3], properties {inputVolume [1,0.3]}}
 {id=QualityLevel_0, children [Mixer_0,Mixer_0,WavePlayer_1,WavePlayer_1]}
 {id=WavePlayer_1, properties {bLooping=true, soundWaveAssetPtr=&/Amb_Winter_Day_Loop_opt}}
 {id=WavePlayer_2, properties {bLooping=true, soundWaveAssetPtr=&/Amb_Winter_Day_Loop_r2}}
 {id=WavePlayer_3, properties {bLooping=true, soundWaveAssetPtr=&/Desert_Ambient_Day_01}}
]
properties {
 attenuationSettings=&Fort_Ambient_Athena_Att
 maxDistance=2097152
 soundClassObject=&Fort_Ambient_Outdoors_Athena
 volumeMultiplier=1.3
}
