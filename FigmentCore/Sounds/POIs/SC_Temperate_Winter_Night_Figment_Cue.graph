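# Sound cue that plays a looping winter night outdoor ambience, named for temperate points of interest in FigmentCore.
# A quality-level root switches between two looping waves: the first two levels use the r2 loop and the last two use the opt loop.
# It uses shared Athena ambient attenuation and the outdoor ambient sound class, with boosted volume and a very large max distance.
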SoundCue /FigmentCore/Sounds/POIs/SC_Temperate_Winter_Night_Figment_Cue
root=QualityLevel_0
nodes [
 {id=QualityLevel_0, children [WavePlayer_0,WavePlayer_0,WavePlayer_1,WavePlayer_1]}
 {id=WavePlayer_0, properties {bLooping=true, soundWaveAssetPtr=&/Amb_Winter_Night_Loop_r2}}
 {id=WavePlayer_1, properties {bLooping=true, soundWaveAssetPtr=&/Amb_Winter_Night_Loop_opt}}
]
properties {
 attenuationSettings=&Fort_Ambient_Athena_Att
 maxDistance=2097152
 soundClassObject=&Fort_Ambient_Outdoors_Athena
 volumeMultiplier=1.5
}
