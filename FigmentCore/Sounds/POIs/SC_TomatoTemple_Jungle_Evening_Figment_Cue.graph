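# A looping evening ambience sound cue for the Tomato Temple jungle POI in FigmentCore.
# A quality-level node sends its first two levels to the Tomato Temple jungle evening loop.
# The last two levels instead play an optimized-named grassland evening loop through a modulator that lowers pitch and slightly varies volume.
# The cue plays at reduced volume using Athena outdoor ambient sound class and ambient attenuation.
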SoundCue /FigmentCore/Sounds/POIs/SC_TomatoTemple_Jungle_Evening_Figment_Cue
root=QualityLevel_0
nodes [
 {id=Modulator_0, children [WavePlayer_4], properties {pitchMax=0.9, pitchMin=0.85, volumeMax=1.1, volumeMin=0.9}}
 {id=QualityLevel_0, children [WavePlayer_0,WavePlayer_0,Modulator_0,Modulator_0]}
 {id=WavePlayer_0, properties {bLooping=true, soundWaveAssetPtr=/FigmentCore/Sounds/POIs/Wavs/Ambient_TomatoTempleJungle_Evening}}
 {id=WavePlayer_4, properties {bLooping=true, soundWaveAssetPtr=/FigmentCore/Sounds/POIs/Wavs/Grassland_Evening_opt}}
]
properties {
 attenuationSettings=&Fort_Ambient_Athena_Att
 maxDistance=2097152
 soundClassObject=&Fort_Ambient_Outdoors_Athena
 volumeMultiplier=0.5
}
